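# Ambient spot sound cue for a huge flapping flag at the Megalodon POI, routed as outdoor ambience.
# Two looped cloth and sail beds pass through nested looping envelopers that drift volume and pitch on differing cycle lengths.
# Looped random, delayed one-shots add modulated sail flaps, cloth flaps and sparser halyard pole impacts.
# Each layer has its own attenuation before the final mix, which feeds only the first two quality-level slots.
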SoundCue /Game/Sounds/Helios/POI/Megalodon/Spot/Spot_Megalodon_Huge_Flag_Flap_loop_Cue
root=QualityLevel_1
nodes [
 {
  id=Attenuation_0
  children [Mixer_1]
  properties {
   bOverrideAttenuation=true
   attenuationOverrides {
    bEnableListenerFocus=true
    falloffDistance=2500
    manualReverbSendLevel=10
    nonFocusVolumeAttenuation=0.5
    reverbSendMethod=Manual
    attenuationShapeExtents {x=1000}
   }
  }
 }
 {
  id=Attenuation_2
  children [Looping_0]
  properties {bOverrideAttenuation=true, attenuationOverrides {falloffDistance=2500, attenuationShapeExtents {x=0}}}
 }
 {
  id=Attenuation_3
  children [Delay_0]
  properties {
   bOverrideAttenuation=true
   attenuationOverrides {
    bEnableListenerFocus=true
    falloffDistance=2500
    manualReverbSendLevel=10
    nonFocusVolumeAttenuation=0.75
    reverbSendMethod=Manual
    attenuationShapeExtents {x=1000}
   }
  }
 }
 {id=Delay_0, children [Enveloper_4], properties {delayMax=0.75, delayMin=0.2}}
 {id=Delay_1, children [Modulator_0], properties {delayMax=18, delayMin=8}}
 {id=Delay_2, children [Random_2], properties {delayMax=5, delayMin=3}}
 {id=Delay_3, children [Random_3], properties {delayMax=5, delayMin=2}}
 {
  id=Enveloper_0
  children [Mixer_2]
  properties {
   bLoop=true
   bLoopIndefinitely=true
   loopEnd=60
   pitchCurve {editorCurveData {keys [
      {interpMode=RCIM_Cubic, time=0}
      <at 15 0.85 RCIM_Cubic>
      {arriveTangent=0.0091, interpMode=RCIM_Cubic, leaveTangent=0.0091, tangentMode=RCTM_User, time=30}
      <at 45 1.1 RCIM_Cubic>
      {time=60}
     ]}}
   volumeCurve {editorCurveData {keys [<at 0 1>]}}
  }
 }
 {
  id=Enveloper_3
  children [Enveloper_0]
  properties {
   bLoop=true
   bLoopIndefinitely=true
   loopEnd=60
   pitchCurve {editorCurveData {keys [{time=0}]}}
   volumeCurve {editorCurveData {keys [
      {arriveTangent=-0.01, interpMode=RCIM_Cubic, time=0, value=1}
      {arriveTangent=-0.0035, interpMode=RCIM_Cubic, leaveTangent=-0.0035, tangentMode=RCTM_User, time=5.0017, value=0.3634}
      {arriveTangent=0.0016, interpMode=RCIM_Cubic, leaveTangent=0.0016, tangentMode=RCTM_User, time=7.591, value=0.8382}
      {arriveTangent=0.0008, interpMode=RCIM_Cubic, leaveTangent=0.0008, tangentMode=RCTM_User, time=11.5666, value=0.3768}
      {arriveTangent=0.0016, interpMode=RCIM_Cubic, leaveTangent=0.0016, tangentMode=RCTM_User, time=13.4364, value=0.3624}
      {arriveTangent=0.0009, interpMode=RCIM_Cubic, leaveTangent=0.0009, time=15.1685, value=0.7207}
      {arriveTangent=-0.0011, interpMode=RCIM_Cubic, leaveTangent=-0.0011, tangentMode=RCTM_User, time=19.0457, value=0.3674}
      {arriveTangent=0.0002, interpMode=RCIM_Cubic, leaveTangent=0.0002, time=24.1334, value=0.8568}
      {arriveTangent=0.0016, interpMode=RCIM_Cubic, leaveTangent=0.0016, tangentMode=RCTM_User, time=28.493, value=0.3696}
      {arriveTangent=-0.0003, interpMode=RCIM_Cubic, leaveTangent=-0.0003, time=31.406, value=0.7353}
      {arriveTangent=0.0009, interpMode=RCIM_Cubic, leaveTangent=0.0009, tangentMode=RCTM_User, time=35.3816, value=0.3676}
      {arriveTangent=-0.0008, interpMode=RCIM_Cubic, leaveTangent=-0.0008, time=38.1075, value=0.8568}
      <at 42.4671 0.3623>
      {arriveTangent=0.0024, interpMode=RCIM_Cubic, leaveTangent=0.0024, tangentMode=RCTM_User, time=45.0257, value=0.3674}
      {arriveTangent=-0.0006, interpMode=RCIM_Cubic, leaveTangent=-0.0006, time=48.1356, value=0.8382}
      {arriveTangent=-0.0056, interpMode=RCIM_Cubic, leaveTangent=-0.0056, tangentMode=RCTM_User, time=53.5932, value=0.3624}
      <at 55.5199 0.7454>
      {arriveTangent=-0.0003, interpMode=RCIM_Cubic, leaveTangent=-0.0003, tangentMode=RCTM_User, time=58.2125, value=0.3621}
      <at 60 1 RCIM_Cubic>
     ]}}
  }
 }
 {
  id=Enveloper_4
  children [Enveloper_3]
  properties {
   bLoop=true
   bLoopIndefinitely=true
   loopEnd=40
   pitchCurve {editorCurveData {keys [{time=0}]}}
   volumeCurve {editorCurveData {keys [
      {arriveTangent=-0.01, interpMode=RCIM_Cubic, time=0, value=1}
      <at 10 0.7 RCIM_Cubic>
      {arriveTangent=-0.0173, interpMode=RCIM_Cubic, leaveTangent=-0.0173, time=20, value=1}
      {arriveTangent=-0.0749, interpMode=RCIM_Cubic, leaveTangent=-0.0749, tangentMode=RCTM_User, time=25.9279, value=0.4252}
      {arriveTangent=0.129, interpMode=RCIM_Cubic, leaveTangent=0.129, tangentMode=RCTM_User, time=36.1629, value=0.5685}
      <at 40 1 RCIM_Cubic>
     ]}}
  }
 }
 {id=Looping_0, children [Delay_1]}
 {id=Looping_1, children [Modulator_1]}
 {id=Looping_2, children [Modulator_2]}
 {id=Mixer_0, children [Attenuation_0, Attenuation_3, Attenuation_2], properties {inputVolume [0.6,1.5,0.35]}}
 {id=Mixer_1, children [Looping_2,Looping_1], properties {inputVolume [1,1.2]}}
 {id=Mixer_2, children [WavePlayer_18, WavePlayer_0], properties {inputVolume [1,1]}}
 {id=Modulator_0, children [Random_1], properties {volumeMin=0.4}}
 {id=Modulator_1, children [Delay_2], properties {pitchMin=0.85, volumeMax=1.2, volumeMin=0.9}}
 {id=Modulator_2, children [Delay_3], properties {pitchMin=0.85, volumeMax=1.2, volumeMin=0.9}}
 {id=QualityLevel_1, children [Mixer_0,Mixer_0,null,null]}
 {
  id=Random_1
  children [
   WavePlayer_21
   WavePlayer_22
   WavePlayer_23
   WavePlayer_24
   WavePlayer_25
   WavePlayer_26
   WavePlayer_27
   WavePlayer_28
   WavePlayer_29
   WavePlayer_30
   WavePlayer_31
   WavePlayer_32
   WavePlayer_33
   WavePlayer_34
   WavePlayer_35
   WavePlayer_36
   WavePlayer_37
   WavePlayer_38
   WavePlayer_39
  ]
  properties {weights [1,1,1,1,1,1,1,1,1,1,1,1,1,1,1,1,1,1,1]}
 }
 {
  id=Random_2
  children [
   WavePlayer_1
   WavePlayer_2
   WavePlayer_3
   WavePlayer_4
   WavePlayer_5
   WavePlayer_6
   WavePlayer_7
   WavePlayer_8
   WavePlayer_9
   WavePlayer_10
   WavePlayer_11
  ]
  properties {weights [1,1,1,1,1,1,1,1,1,1,1]}
 }
 {
  id=Random_3
  children [WavePlayer_12, WavePlayer_13, WavePlayer_14, WavePlayer_15, WavePlayer_16, WavePlayer_17]
  properties {weights [1,1,1,1,1,1]}
 }
 {id=WavePlayer_0, properties {bLooping=true, soundWaveAssetPtr=/Game/Athena/Sounds/Ambient/PirateCove/Amb_PirateCove_SailFlap_01}}
 {id=WavePlayer_1, properties {soundWaveAssetPtr=/Game/Sounds/Artemis/POI/Chrome_BugleBay/Spot_PirateCove_SailFlap_Large_01}}
 {id=WavePlayer_10, properties {soundWaveAssetPtr=/Game/Sounds/Artemis/POI/Chrome_BugleBay/Spot_PirateCove_SailFlap_Large_10}}
 {id=WavePlayer_11, properties {soundWaveAssetPtr=/Game/Sounds/Artemis/POI/Chrome_BugleBay/Spot_PirateCove_SailFlap_Large_11}}
 {id=WavePlayer_12, properties {soundWaveAssetPtr=/Game/Sounds/Helios/Environment_Actors/FlagBanner/Spot_Cloth_Flap_Oneshot_06}}
 {id=WavePlayer_13, properties {soundWaveAssetPtr=/Game/Sounds/Helios/Environment_Actors/FlagBanner/Spot_Cloth_Flap_Oneshot_05}}
 {id=WavePlayer_14, properties {soundWaveAssetPtr=/Game/Sounds/Helios/Environment_Actors/FlagBanner/Spot_Cloth_Flap_Oneshot_04}}
 {id=WavePlayer_15, properties {soundWaveAssetPtr=/Game/Sounds/Helios/Environment_Actors/FlagBanner/Spot_Cloth_Flap_Oneshot_03}}
 {id=WavePlayer_16, properties {soundWaveAssetPtr=/Game/Sounds/Helios/Environment_Actors/FlagBanner/Spot_Cloth_Flap_Oneshot_02}}
 {id=WavePlayer_17, properties {soundWaveAssetPtr=/Game/Sounds/Helios/Environment_Actors/FlagBanner/Spot_Cloth_Flap_Oneshot_01}}
 {
  id=WavePlayer_18
  properties {bLooping=true, soundWaveAssetPtr=/Game/Sounds/Helios/Environment_Actors/FlagBanner/Spot_Flag_Flap_Cloth_Banner_Slow_Loop}
 }
 {id=WavePlayer_2, properties {soundWaveAssetPtr=/Game/Sounds/Artemis/POI/Chrome_BugleBay/Spot_PirateCove_SailFlap_Large_02}}
 {id=WavePlayer_21, properties {soundWaveAssetPtr=/Game/Sounds/Helios/POI/MarinaTown/Spots/Spot_Chaparral_Halyard_Pole_Impact_19}}
 {id=WavePlayer_22, properties {soundWaveAssetPtr=/Game/Sounds/Helios/POI/MarinaTown/Spots/Spot_Chaparral_Halyard_Pole_Impact_18}}
 {id=WavePlayer_23, properties {soundWaveAssetPtr=/Game/Sounds/Helios/POI/MarinaTown/Spots/Spot_Chaparral_Halyard_Pole_Impact_17}}
 {id=WavePlayer_24, properties {soundWaveAssetPtr=/Game/Sounds/Helios/POI/MarinaTown/Spots/Spot_Chaparral_Halyard_Pole_Impact_16}}
 {id=WavePlayer_25, properties {soundWaveAssetPtr=/Game/Sounds/Helios/POI/MarinaTown/Spots/Spot_Chaparral_Halyard_Pole_Impact_15}}
 {id=WavePlayer_26, properties {soundWaveAssetPtr=/Game/Sounds/Helios/POI/MarinaTown/Spots/Spot_Chaparral_Halyard_Pole_Impact_14}}
 {id=WavePlayer_27, properties {soundWaveAssetPtr=/Game/Sounds/Helios/POI/MarinaTown/Spots/Spot_Chaparral_Halyard_Pole_Impact_13}}
 {id=WavePlayer_28, properties {soundWaveAssetPtr=/Game/Sounds/Helios/POI/MarinaTown/Spots/Spot_Chaparral_Halyard_Pole_Impact_12}}
 {id=WavePlayer_29, properties {soundWaveAssetPtr=/Game/Sounds/Helios/POI/MarinaTown/Spots/Spot_Chaparral_Halyard_Pole_Impact_11}}
 {id=WavePlayer_3, properties {soundWaveAssetPtr=/Game/Sounds/Artemis/POI/Chrome_BugleBay/Spot_PirateCove_SailFlap_Large_03}}
 {id=WavePlayer_30, properties {soundWaveAssetPtr=/Game/Sounds/Helios/POI/MarinaTown/Spots/Spot_Chaparral_Halyard_Pole_Impact_10}}
 {id=WavePlayer_31, properties {soundWaveAssetPtr=/Game/Sounds/Helios/POI/MarinaTown/Spots/Spot_Chaparral_Halyard_Pole_Impact_09}}
 {id=WavePlayer_32, properties {soundWaveAssetPtr=/Game/Sounds/Helios/POI/MarinaTown/Spots/Spot_Chaparral_Halyard_Pole_Impact_08}}
 {id=WavePlayer_33, properties {soundWaveAssetPtr=/Game/Sounds/Helios/POI/MarinaTown/Spots/Spot_Chaparral_Halyard_Pole_Impact_07}}
 {id=WavePlayer_34, properties {soundWaveAssetPtr=/Game/Sounds/Helios/POI/MarinaTown/Spots/Spot_Chaparral_Halyard_Pole_Impact_06}}
 {id=WavePlayer_35, properties {soundWaveAssetPtr=/Game/Sounds/Helios/POI/MarinaTown/Spots/Spot_Chaparral_Halyard_Pole_Impact_05}}
 {id=WavePlayer_36, properties {soundWaveAssetPtr=/Game/Sounds/Helios/POI/MarinaTown/Spots/Spot_Chaparral_Halyard_Pole_Impact_04}}
 {id=WavePlayer_37, properties {soundWaveAssetPtr=/Game/Sounds/Helios/POI/MarinaTown/Spots/Spot_Chaparral_Halyard_Pole_Impact_03}}
 {id=WavePlayer_38, properties {soundWaveAssetPtr=/Game/Sounds/Helios/POI/MarinaTown/Spots/Spot_Chaparral_Halyard_Pole_Impact_02}}
 {id=WavePlayer_39, properties {soundWaveAssetPtr=/Game/Sounds/Helios/POI/MarinaTown/Spots/Spot_Chaparral_Halyard_Pole_Impact_01}}
 {id=WavePlayer_4, properties {soundWaveAssetPtr=/Game/Sounds/Artemis/POI/Chrome_BugleBay/Spot_PirateCove_SailFlap_Large_04}}
 {id=WavePlayer_5, properties {soundWaveAssetPtr=/Game/Sounds/Artemis/POI/Chrome_BugleBay/Spot_PirateCove_SailFlap_Large_05}}
 {id=WavePlayer_6, properties {soundWaveAssetPtr=/Game/Sounds/Artemis/POI/Chrome_BugleBay/Spot_PirateCove_SailFlap_Large_06}}
 {id=WavePlayer_7, properties {soundWaveAssetPtr=/Game/Sounds/Artemis/POI/Chrome_BugleBay/Spot_PirateCove_SailFlap_Large_07}}
 {id=WavePlayer_8, properties {soundWaveAssetPtr=/Game/Sounds/Artemis/POI/Chrome_BugleBay/Spot_PirateCove_SailFlap_Large_08}}
 {id=WavePlayer_9, properties {soundWaveAssetPtr=/Game/Sounds/Artemis/POI/Chrome_BugleBay/Spot_PirateCove_SailFlap_Large_09}}
]
properties {
 maxDistance=3500
 soundClassObject=&Fort_Ambient_Outdoors_Athena
 volumeMultiplier=0.501
 attenuationOverrides {
  attenuationShape=Capsule
  bEnableFocusInterpolation=true
  bEnableOcclusion=true
  falloffDistance=2500
  focusAttackInterpSpeed=3
  focusReleaseInterpSpeed=3
  manualReverbSendLevel=10
  nonFocusVolumeAttenuation=0.5
  occlusionLowPassFilterFrequency=3000
  occlusionTraceChannel=ECC_GameTraceChannel11
  reverbDistanceMax=2000
  reverbSendMethod=Manual
  reverbWetLevelMax=5
  attenuationShapeExtents {x=0}
 }
 busSends [{audioBus=null, minSendLevel=1, sendLevel=1, soundSourceBus=&AmbientAudio_Exterior_SourceBus}]
 soundSubmixSends [{minSendLevel=0, soundSubmix=&Fortnite_Reverb_Default}]
}
