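# Ambient spot cue for the Megalodon POI that plays intermittent distant metal groans through a cone-shaped attenuation override.
# A loop waits a random delay, then picks one of four pitch-modulated, level-mixed pools: two identical cave wronk sets, heavy shuffles and bridge stress groans.
# Two nested looping Envelopers shape volume and pitch over time, and every quality level shares that random selection.
# Occlusion, distance low-pass, listener focus and limited concurrency control playback.
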SoundCue /Game/Sounds/Helios/POI/Megalodon/Spot/Spot_Megalodon_Metal_Groan_Cone_Cue
root=Enveloper_1
nodes [
 {id=Delay_0, children [QualityLevel_0], properties {delayMax=18, delayMin=8}}
 {
  id=Enveloper_1
  children [Enveloper_3]
  properties {
   bLoop=true
   bLoopIndefinitely=true
   loopEnd=30
   pitchCurve {editorCurveData {keys [
      {interpMode=RCIM_Cubic, time=0}
      {arriveTangent=-0.0001, interpMode=RCIM_Cubic, leaveTangent=-0.0001, tangentMode=RCTM_User, time=6}
      {interpMode=RCIM_Cubic, tangentMode=RCTM_User, time=14, value=0.8}
      <at 30 0.8 RCIM_Cubic>
     ]}}
   volumeCurve {editorCurveData {keys [<at 0 1>]}}
  }
 }
 {
  id=Enveloper_3
  children [Looping_0]
  properties {
   bLoop=true
   bLoopIndefinitely=true
   loopEnd=20
   pitchCurve {editorCurveData {keys [{time=0}]}}
   volumeCurve {editorCurveData {keys [
      {arriveTangent=-0.01, interpMode=RCIM_Cubic, leaveTangent=-0.01, time=0, value=1}
      <at 10 0.9 RCIM_Cubic>
      {arriveTangent=0.01, interpMode=RCIM_Cubic, leaveTangent=0.01, time=20, value=1}
     ]}}
  }
 }
 {id=Looping_0, children [Delay_0]}
 {id=Mixer_1, children [Modulator_1], properties {inputVolume [0.4]}}
 {id=Mixer_2, children [Modulator_0], properties {inputVolume [0.4]}}
 {id=Mixer_3, children [Modulator_2], properties {inputVolume [1]}}
 {id=Mixer_4, children [Modulator_3], properties {inputVolume [0.6]}}
 {id=Modulator_0, children [Random_0], properties {pitchMin=0.9}}
 {id=Modulator_1, children [Random_2], properties {pitchMin=0.9}}
 {id=Modulator_2, children [Random_3], properties {pitchMin=0.9}}
 {id=Modulator_3, children [Random_4], properties {pitchMin=0.9}}
 {id=QualityLevel_0, children [Random_1,Random_1,Random_1,Random_1]}
 {id=Random_0, children [WavePlayer_0,WavePlayer_1,WavePlayer_2,WavePlayer_4,WavePlayer_7], properties {weights [1,1,1,1,1]}}
 {id=Random_1, children [Mixer_2,Mixer_1,Mixer_3,Mixer_4], properties {weights [1,1,1,1]}}
 {id=Random_2, children [WavePlayer_8, WavePlayer_9, WavePlayer_10, WavePlayer_11, WavePlayer_12], properties {weights [1,1,1,1,1]}}
 {id=Random_3, children [WavePlayer_13, WavePlayer_14, WavePlayer_15, WavePlayer_16], properties {weights [1,1,1,1]}}
 {
  id=Random_4
  children [WavePlayer_17, WavePlayer_18, WavePlayer_19, WavePlayer_20, WavePlayer_21, WavePlayer_22, WavePlayer_23]
  properties {weights [1,1,1,1,1,1,1]}
 }
 {id=WavePlayer_0, properties {soundWaveAssetPtr=/Game/Sounds/Artemis/POI/Hollow_Base/Spots/Spot_Cave_Distant_Metal_Wronk_1}}
 {id=WavePlayer_1, properties {soundWaveAssetPtr=/Game/Sounds/Artemis/POI/Hollow_Base/Spots/Spot_Cave_Distant_Metal_Wronk_2}}
 {id=WavePlayer_10, properties {soundWaveAssetPtr=/Game/Sounds/Artemis/POI/Hollow_Base/Spots/Spot_Cave_Distant_Metal_Wronk_3}}
 {id=WavePlayer_11, properties {soundWaveAssetPtr=/Game/Sounds/Artemis/POI/Hollow_Base/Spots/Spot_Cave_Distant_Metal_Wronk_5}}
 {id=WavePlayer_12, properties {soundWaveAssetPtr=/Game/Sounds/Artemis/POI/Hollow_Base/Spots/Spot_Cave_Distant_Metal_Wronk_8}}
 {id=WavePlayer_13, properties {soundWaveAssetPtr=/Game/Sounds/Asteria/Environment_Actors/MetalStress/Spot_Metal_Heavy_Shuffle_01}}
 {id=WavePlayer_14, properties {soundWaveAssetPtr=/Game/Sounds/Asteria/Environment_Actors/MetalStress/Spot_Metal_Heavy_Shuffle_02}}
 {id=WavePlayer_15, properties {soundWaveAssetPtr=/Game/Sounds/Asteria/Environment_Actors/MetalStress/Spot_Metal_Heavy_Shuffle_03}}
 {id=WavePlayer_16, properties {soundWaveAssetPtr=/Game/Sounds/Asteria/Environment_Actors/MetalStress/Spot_Metal_Heavy_Shuffle_04}}
 {id=WavePlayer_17, properties {soundWaveAssetPtr=/Game/Sounds/Artemis/POI/Big_Bridge/Spots/Spot_Big_Metal_Bridge_Huge_Stress_Groan_01}}
 {id=WavePlayer_18, properties {soundWaveAssetPtr=/Game/Sounds/Artemis/POI/Big_Bridge/Spots/Spot_Big_Metal_Bridge_Huge_Stress_Groan_02}}
 {id=WavePlayer_19, properties {soundWaveAssetPtr=/Game/Sounds/Artemis/POI/Big_Bridge/Spots/Spot_Big_Metal_Bridge_Huge_Stress_Groan_03}}
 {id=WavePlayer_2, properties {soundWaveAssetPtr=&Spot_Cave_Distant_Metal_Wronk_3}}
 {id=WavePlayer_20, properties {soundWaveAssetPtr=/Game/Sounds/Artemis/POI/Big_Bridge/Spots/Spot_Big_Metal_Bridge_Huge_Stress_Groan_04}}
 {id=WavePlayer_21, properties {soundWaveAssetPtr=/Game/Sounds/Artemis/POI/Big_Bridge/Spots/Spot_Big_Metal_Bridge_Huge_Stress_Groan_05}}
 {id=WavePlayer_22, properties {soundWaveAssetPtr=/Game/Sounds/Artemis/POI/Big_Bridge/Spots/Spot_Big_Metal_Bridge_Huge_Stress_Groan_06}}
 {id=WavePlayer_23, properties {soundWaveAssetPtr=/Game/Sounds/Artemis/POI/Big_Bridge/Spots/Spot_Big_Metal_Bridge_Huge_Stress_Groan_07}}
 {id=WavePlayer_4, properties {soundWaveAssetPtr=&Spot_Cave_Distant_Metal_Wronk_5}}
 {id=WavePlayer_7, properties {soundWaveAssetPtr=&Spot_Cave_Distant_Metal_Wronk_8}}
 {id=WavePlayer_8, properties {soundWaveAssetPtr=&Spot_Cave_Distant_Metal_Wronk_1}}
 {id=WavePlayer_9, properties {soundWaveAssetPtr=&Spot_Cave_Distant_Metal_Wronk_2}}
]
properties {
 attenuationSettings=/Game/Sounds/Attenuation/Fort_TileAmbient_OneShot_Distant_Creak_Att
 bEnableSubmixSends=false
 bOverrideAttenuation=true
 bOverrideConcurrency=true
 maxDistance=13000
 soundClassObject=&Fort_Ambient_Outdoors_Athena
 volumeMultiplier=0.2
 attenuationOverrides {
  attenuationShape=Cone
  bAttenuateWithLPF=true
  bEnableListenerFocus=true
  bEnableOcclusion=true
  bEnableSubmixSends=true
  falloffDistance=9000
  lPFFrequencyAtMax=1000
  lPFRadiusMin=2000
  nonFocusVolumeAttenuation=0.75
  occlusionInterpolationTime=2
  occlusionLowPassFilterFrequency=1000
  occlusionTraceChannel=ECC_GameTraceChannel11
  occlusionVolumeAttenuation=0.5
  stereoSpread=1000
  attenuationShapeExtents=<4000 80>
  submixSendSettings [{
    bEnableHPFCutoff=false
    bEnableLPFCutoff=false
    disableManualSendClamp=false
    hPFCutoff=20
    lPFCutoff=20000
    maxSendDistance=6000
    maxSendLevel=1
    minSendDistance=2000
    minSendLevel=0.6
    sendLevel=0.2
    sendLevelControlMethod=Linear
    soundSubmix=&Fortnite_Reverb_Distant
    customSendLevelCurve {editorCurveData {keys [<at 0 0.6>,<at 1 1>]}}
   }]
 }
 busSends [{
   audioBus=null
   maxSendDistance=6000
   minSendDistance=2000
   minSendLevel=1
   sendLevel=1
   soundSourceBus=&AmbientAudio_Exterior_SourceBus
   customSendLevelCurve {editorCurveData {keys [<at 0 0.6>,<at 1 1.5>]}}
  }]
 concurrencyOverrides {maxCount=2, voiceStealReleaseTime=5}
}
